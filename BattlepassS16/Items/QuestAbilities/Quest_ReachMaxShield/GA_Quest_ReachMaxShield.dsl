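; Passive quest ability that credits the owning player's quest for reaching a shield threshold.
; On activation it checks the starting shield, then waits for attribute threshold changes, storing each result before calling TryProgressQuest.
; If the threshold is met, TryProgressQuest rechecks shield after a brief delay before sending the Athena.Quests.MaxShield stat event.
; The delay guards against momentary shield spikes seen in an LTM bus jump (FORT-251834).
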(event Ability|EventFailedToActivatePassiveAbility
  (params (ActorInfo :type Gameplay%20Ability%20Actor%20Info%20Structure)))

(event Ability|EventActivateAbility
  (bind (_bsuccessfullyfoundattribute _returnvalue) (Ability|Attribute|GetFloatAttribute (Variables|GAVars|Data|GetOwningPlayerPawn) (Variables|GAVars|Setup|GetAtt_Shield)))
  ;[1 1 1] Check starting shield value in case player spawns with threshold success
  (|CallParentFunction)
  (bind _thresholdsuccess (CallFunction|CheckInitialAttributeValue _bsuccessfullyfoundattribute _returnvalue))
  (Variables|GAVars|Data|SetShieldSuccess _thresholdsuccess)
  (CallFunction|TryProgressQuest)
  ;[1 1 1] Wait for attribute changes and check progress again
  (bind (_asynctaskproxy _bmatchescomparison) (Ability|Tasks|WaitForAttributeChangeThreshold (Variables|GAVars|Setup|GetAtt_Shield) "GreaterThanOrEqualTo" (Variables|GAVars|Setup|GetComparisonValue))
    (:OnChange
      (Variables|GAVars|Data|SetShieldSuccess _bmatchescomparison)
      (CallFunction|TryProgressQuest))))

(event Custom|TryProgressQuest
  (bind (_bsuccessfullyfoundattribute _returnvalue) (Ability|Attribute|GetFloatAttribute (Variables|GAVars|Data|GetOwningPlayerPawn) (Variables|GAVars|Setup|GetAtt_Shield)))
  ;[1 1 1] If shield meets threshold, send quest progress
  (if (Variables|GAVars|Data|GetShieldSuccess)
    ;[1 0.8959 0.3865] Added this delay and additional check to prevent future occurrences of FORT-251834 In that case, the LTM flashed the HP and Shield to 100 briefly when the player jumped from the bus
    (Utilities|FlowControl|Delay 0.05)
    (if (>= _returnvalue (Variables|GAVars|Setup|GetComparisonValue))
      (Quest|SendComplexCustomStatEvent (Variables|GAVars|Data|GetOwningPlayerQuestManager) 0 "(GameplayTags=((TagName=\"Athena.Quests.MaxShield\")))" "(GameplayTags=)" 1))))

(fn CheckInitialAttributeValue
  (params (AttributeFound :type Boolean) (Value :type Float%20%28double-precision%29))
  (if AttributeFound
    _))
